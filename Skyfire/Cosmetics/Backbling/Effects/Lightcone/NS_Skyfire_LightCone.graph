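# Niagara light cone effect for the Skyfire back bling, tinted by User.LootLevel_Color and attached to User.SkeletalMesh.
# Active emitters layer glow sprites with a light, cone and soft-ray meshes, and continuously spawned randomized rays.
# User.Alpha and an optional User.BlinkSpeed sine pulse (TurnOnBlink) drive UserColor alpha; Cone_Ray also scales with Scale_Size and time of day.
# Light, LightCone, LightCone001 and LightCone_mesh emitters are disabled.
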NiagaraSystem /Skyfire/Cosmetics/Backbling/Effects/Lightcone/NS_Skyfire_LightCone
dependencies {
 usedDataInterfaces [&NiagaraDataInterfaceCurve]
 usedDynamicInputs [&MakeIntFromFloat]
 usedModules [
  &SystemState
  &self:SystemSpawnScript.SystemScriptSource.SystemScriptGraph.NiagaraNodeAssignment_0.SetVariables_#1
 ]
}
emitters {
 BaseLight {
  emitterUpdateScript [
   &EmitterState()
   &SpawnBurst_Instantaneous(Spawn Count=1, Use Spawn Probability=false)
   /SetVariables(Emitter.Color:LinearColor=&MultiplyLinearColor(
     Linear Color=&LerpLinearColors(StartColor=$User.LootLevel_Color, EndColor=<rgba 1 1 1 1>, LerpFactor=0.15)
     Scale Factor=<4 4 4 1>
    )) {isSetParameters=true, module=SetVariables_#2}
   /SetVariables(Emitter.OpacityOverride:NiagaraFloat=&Multiply_Float(
     A=0.5
     B=&Add_Float(
      A=1
      B=&Sine(
       Normalized Angle=&Multiply_Float(A=$Engine.Time, B=&Multiply_Float(A=$User.BlinkSpeed, B=1))
       Period=1
       Scale=1
       Bias=0
      )
     )
    )) {isSetParameters=true, module=SetVariables_#3}
  ]
  particleSpawnScript [
   /Niagara/Modules/Spawn/Initialization/InitializeParticle(
    Lifetime=7
    Module.Write Color=true
    Module.Write Lifetime=true
    Module.Write Mass=true
    Module.Write Material Random=true
    Module.Write Position=false
    Module.Write Scale=true
    Module.Write SpriteRotation=false
    Module.Write SpriteSize=true
    Sprite Size=<35 35>
   )
   &Color(Color=$User.LootLevel_Color) {enabled=false}
   &InitialMeshOrientation(Orientation Axis=<1 0 1>, Orientation Vector=<0 0 1>, Use Orientation Vector=true, Use Rotation Vector=false)
  ]
  particleUpdateScript [
   &ParticleState()
   &CameraOffset(Camera Offset Amount=5)
   &DynamicMaterialParameters(
    Index 0 Param 1=0.32
    Index 0 Param 2=20
    Index 0 Param 3=1
    Index 0 Param 4=1
    Module.Param0WriteEnabled=true
    Module.Param1WriteEnabled=true
    Module.Param2WriteEnabled=true
    Module.Param3WriteEnabled=true
   )
   &SkeletalMeshLocation(
    MaskSpawnGroup=false
    Mesh Sampling Type=::NewEnumerator5
    Sampled Position Offset=<0 0.2 0>
    Skeletal Mesh=$User.SkeletalMesh
    Socket Sampling Mode=::NewEnumerator5
   )
   /SetVariables(Particles.SpriteFacing=<-0.3 0 1>) {isSetParameters=true, module=SetVariables_#4}
   /SetVariables(Particles.UserColor:LinearColor=&MakeLinearColorFromVectorAndFloat(
     "Vector (RGB)"=<1 1 1>
     "Float (Alpha)"=&Multiply_Float(
      A=&Add_Float(
       A=&Multiply_Float(
        A=&Multiply_Float(
         A=&FloatFromCurve(
          FloatCurve=DataInterface() {propertyValues {Curve {keys [{interpMode=RCIM_Cubic, tangentMode=RCTM_User, time=0.8, value=1}, {interpMode=RCIM_Cubic, tangentMode=RCTM_User, time=1, value=0}]}}}
          Scale Curve=2
         )
         B=&MakeCustomFloatFromBool(Bool=&SetBoolByFloatComparison(A=$User.TurnOnBlink, B=0), True Float=$Emitter.OpacityOverride, False Float=1)
        )
        B=&ClampFloat(Float=$User.Alpha)
       )
       B=0.1
      )
      B=&ClampFloat(Float=$User.Alpha)
     )
    )) {isSetParameters=true, module=SetVariables_#5}
   &AlignSpriteToMeshOrientation(Mesh Orientation Relative Sprite Alignment Vector=<1 0 0>, Mesh Orientation Relative Sprite Facing Vector:Vector3f=<0 0 1>) {enabled=false}
   &FN_Light_Attributes(
    Color=&MultiplyLinearColor(
     Linear Color=$User.LootLevel_Color
     Scale Factor=&MakeVector4_FromFloat(Float=&FloatFromCurve(FloatCurve=DataInterface() {propertyValues {Curve {keys [
           {tangentMode=RCTM_User, time=0, value=1}
           {arriveTangent=-0.0405, interpMode=RCIM_Cubic, leaveTangent=-0.0405, tangentMode=RCTM_User, time=0.7009, value=1}
           {arriveTangent=-8.0089, interpMode=RCIM_Cubic, leaveTangent=-8.0089, tangentMode=RCTM_User, time=1, value=0}
          ]}}}))
    )
    Create Color Attribute=true
    Create Exponent Attribute=true
    Create Position Attribute=true
    Create Radius Attribute=true
    Exponent=1
    Position=&Add_Vector(A=Linked() {linkedVariable {name=Particles.Position, type=NiagaraPosition}}, B=<0 0 10>)
    Radius=20
    Scale Color=30
   )
  ]
  rendererClasses [&NiagaraSpriteRendererProperties, &NiagaraLightRendererProperties]
  renderers [
   {rendererClass=&NiagaraSpriteRendererProperties}
   {rendererClass=&NiagaraSpriteRendererProperties, rendererIndex=1}
   {rendererClass=&NiagaraLightRendererProperties, rendererIndex=2}
  ]
 }
 BaseLight_AngleRelevant {
  emitterUpdateScript [
   &EmitterState()
   &SpawnBurst_Instantaneous(Spawn Count=1, Use Spawn Probability=false)
   /SetVariables(Emitter.Color:LinearColor=$User.LootLevel_Color) {isSetParameters=true, module=SetVariables_#2}
   /SetVariables(Emitter.OpacityOverride:NiagaraFloat=&Multiply_Float(
     A=0.5
     B=&Add_Float(
      A=1
      B=&Sine(
       Normalized Angle=&Multiply_Float(A=$Engine.Time, B=&Multiply_Float(A=$User.BlinkSpeed, B=1))
       Period=1
       Scale=1
       Bias=0
      )
     )
    )) {isSetParameters=true, module=SetVariables_#3}
  ]
  particleSpawnScript [
   &InitializeParticle(
    Lifetime=7
    Module.Write Color=true
    Module.Write Lifetime=true
    Module.Write Mass=true
    Module.Write Material Random=true
    Module.Write Position=false
    Module.Write Scale=false
    Module.Write SpriteRotation=false
    Module.Write SpriteSize=true
    Sprite Size=<20 20>
   )
   /SetVariables(Particles.SpriteFacing=<0 0 1>) {isSetParameters=true, module=SetVariables_#4}
   &Color(Color=$User.LootLevel_Color) {enabled=false}
  ]
  particleUpdateScript [
   &ParticleState()
   &CameraOffset(Camera Offset Amount=5)
   &DynamicMaterialParameters(
    Index 0 Param 1=1
    Index 0 Param 2=20
    Index 0 Param 3=1
    Index 0 Param 4=1
    Module.Param0WriteEnabled=true
    Module.Param1WriteEnabled=true
    Module.Param2WriteEnabled=true
    Module.Param3WriteEnabled=true
   )
   &SkeletalMeshLocation(MaskSpawnGroup=false, Mesh Sampling Type=::NewEnumerator5, Skeletal Mesh=$User.SkeletalMesh, Socket Sampling Mode=::NewEnumerator5)
   /SetVariables(Particles.UserColor:LinearColor=&MakeLinearColorFromVectorAndFloat(
     "Vector (RGB)"=<1 1 1>
     "Float (Alpha)"=&Multiply_Float(
      A=&Add_Float(
       A=&Multiply_Float(
        A=&Multiply_Float(
         A=&FloatFromCurve(FloatCurve=DataInterface() {propertyValues {Curve {keys [{interpMode=RCIM_Cubic, tangentMode=RCTM_User, time=0.8, value=1}, {interpMode=RCIM_Cubic, tangentMode=RCTM_User, time=1, value=0}]}}})
         B=&MakeCustomFloatFromBool(Bool=&SetBoolByFloatComparison(A=$User.TurnOnBlink, B=0), True Float=$Emitter.OpacityOverride, False Float=1)
        )
        B=&ClampFloat(Float=$User.Alpha)
       )
       B=0.1
      )
      B=&ClampFloat(Float=$User.Alpha)
     )
    )) {isSetParameters=true, module=SetVariables_#5}
  ]
  rendererClasses [&NiagaraSpriteRendererProperties]
  renderers [{rendererClass=&NiagaraSpriteRendererProperties}]
 }
 Cone_Ray {
  emitterUpdateScript [
   &EmitterState()
   &SpawnBurst_Instantaneous(Spawn Count=1, Use Spawn Probability=false)
   /SetVariables(Emitter.Color:LinearColor=&MultiplyLinearColor(Linear Color=$User.LootLevel_Color, Scale Factor=<1 1 1 1>)) {isSetParameters=true}
   /SetVariables(Emitter.OpacityOverride:NiagaraFloat=&Multiply_Float(
     A=0.5
     B=&Add_Float(
      A=1
      B=&Sine(
       Normalized Angle=&Multiply_Float(A=$Engine.Time, B=&Multiply_Float(A=$User.BlinkSpeed, B=1))
       Period=1
       Scale=1
       Bias=0
      )
     )
    )) {isSetParameters=true}
  ]
  particleSpawnScript [
   /Niagara/Modules/Spawn/Initialization/V2/InitializeParticle(
    Color=<rgba 1 1 1 0.5>
    Lifetime=7
    Mesh Scale=<0.165 0.2 0.2>
    Mesh Scale Mode=::NewEnumerator4
    UsePositionOffset=false
    Write Material Random=true
   )
   &InitialMeshOrientation(Rotation=<0 -0.25 0>, Use Orientation Vector=true, Use Rotation Vector=true)
  ]
  particleUpdateScript [
   &ParticleState()
   &SkeletalMeshLocation(
    MaskSpawnGroup=false
    Mesh Sampling Type=::NewEnumerator5
    Sampled Position Offset=<0 0 -4>
    Skeletal Mesh=$User.SkeletalMesh
    Socket Sampling Mode=::NewEnumerator5
   )
   &SolveForcesAndVelocity(Clamp Velocity=false, Limit Acceleration=false)
   &ScaleColor(
    Scale RGB=&Multiply_VectorByFloat(Vector=<1 1 1>, Float=&FNTimeOfDayScalar(Night Scale=0.5, Morning Scale=0.3, Evening Scale=0.4))
    ScaleA=true
    ScaleRGB=true
   )
   &ScaleMeshSize(Scale Factor=&MakeVector(
     X=1
     Y=&ClampFloat(Float=&Divide_Float(A=$User.Scale_Size, B=40), Min=0.8, Max=1.2)
     Z=&ClampFloat(Float=&Divide_Float(A=$User.Scale_Size, B=40), Min=0.8, Max=1.2)
    ))
   /SetVariables(Particles.UserColor:LinearColor=&MakeLinearColorFromVectorAndFloat(
     "Vector (RGB)"=&MakeVectorFromLinearColorRGB(Color=$User.LootLevel_Color)
     "Float (Alpha)"=&Multiply_Float(
      A=&Add_Float(
       A=&Multiply_Float(
        A=&Multiply_Float(
         A=&FloatFromCurve(FloatCurve=DataInterface() {propertyValues {Curve {keys [{interpMode=RCIM_Cubic, tangentMode=RCTM_User, time=0.8, value=1}, {interpMode=RCIM_Cubic, tangentMode=RCTM_User, time=1, value=0}]}}})
         B=&MakeCustomFloatFromBool(Bool=&SetBoolByFloatComparison(A=$User.TurnOnBlink, B=0), True Float=$Emitter.OpacityOverride, False Float=1)
        )
        B=&ClampFloat(Float=$User.Alpha)
       )
       B=0.1
      )
      B=&ClampFloat(Float=&Multiply_Float(
        A=$User.Alpha
        B=&FloatFromCurve(FloatCurve=DataInterface() {propertyValues {Curve {keys [<at 0 0.5>,<at 0.75 0.5>,<at 1 0>]}}}, CurveIndex=$System.NormalizedLoopAge)
       ))
     )
    )) {isSetParameters=true}
   /Niagara/Modules/Update/Orientation/V2/OrientMeshToVector(Override Quat=false, UseOptionalSideAxis=false, UseOptionalUpAxis=false) {enabled=false}
  ]
  rendererClasses [&NiagaraMeshRendererProperties]
  renderers [
   {rendererClass=&NiagaraMeshRendererProperties, rendererIndex=0}
   {rendererClass=&NiagaraMeshRendererProperties, rendererIndex=1}
  ]
 }
 Light {
  bEnabled=false
  emitterUpdateScript [
   &EmitterState()
   &SpawnBurst_Instantaneous(Spawn Count=1, Use Spawn Probability=false)
  ]
  particleSpawnScript [
   /Niagara/Modules/Spawn/Initialization/InitializeParticle(
    Module.Write Color=true
    Module.Write Lifetime=true
    Module.Write Mass=true
    Module.Write Material Random=true
    Module.Write Position=false
    Module.Write Scale=false
    Module.Write SpriteRotation=false
    Module.Write SpriteSize=true
    Position=<0 0 20>
    Sprite Size=<100 100>
   )
   &SystemLocation(MaskSpawnGroup=false, Offset=<0 0 20>) {enabled=false}
  ]
  particleUpdateScript [
   &FN_Light_Attributes(
    Color=&MultiplyLinearColor(Linear Color=$User.LootLevel_Color, Scale Factor=&MakeVector4_FromFloat(Float=1.5))
    Create Color Attribute=true
    Create Exponent Attribute=true
    Create Position Attribute=true
    Create Radius Attribute=true
    Exponent=3
    Position=<0 0 20>
    Radius=100
    Scale Color=&FloatFromCurve(FloatCurve=DataInterface() {propertyValues {Curve {keys [<at 0 0>,<at 0.1 1>,<at 0.9 1>,<at 1 0>]}}})
   )
   &ParticleState()
   &FN_Gen5_Light(
    Attenutation=80
    Color:LinearColor=&MultiplyLinearColor(Linear Color=$User.LootLevel_Color, Scale Factor=<1 1 1 1>)
    Exponent=5
    Intensity:NiagaraFloat=&Lerp_Float(A=11.5, B=12, Alpha=&Sine(Normalized Angle=$Emitter.Age, Period=0.25, Scale=2, Bias=4))
    Position Offset:Vector3f=<0 0 -20>
   )
   &SkeletalMeshLocation(
    MaskSpawnGroup=false
    Mesh Sampling Type=::NewEnumerator5
    OverrideRandomSeed=false
    Sampled Position Offset=<0 0 20>
    Skeletal Mesh=$User.SkeletalMesh
   )
  ]
  rendererClasses [&NiagaraLightRendererProperties, &NiagaraComponentRendererProperties]
  renderers [
   {rendererClass=&NiagaraLightRendererProperties, rendererIndex=0}
   {rendererClass=&NiagaraComponentRendererProperties, rendererIndex=1}
  ]
 }
 LightCone {
  bEnabled=false
  emitterUpdateScript [
   &EmitterState()
   &SpawnBurst_Instantaneous(Spawn Count=1, Use Spawn Probability=false)
   /SetVariables(Emitter.OpacityOverride:NiagaraFloat=&Multiply_Float(
     A=1
     B=&Add_Float(
      A=1
      B=&Sine(
       Normalized Angle=&Multiply_Float(A=$Engine.Time, B=&Multiply_Float(A=$User.BlinkSpeed, B=1))
       Period=1
       Scale=1
       Bias=0
      )
     )
    )) {isSetParameters=true}
  ]
  particleSpawnScript [
   &InitializeParticle(
    Lifetime=7
    Mesh Scale=<0.1 0.1 1>
    Module.Write Color=false
    Module.Write Lifetime=true
    Module.Write Mass=true
    Module.Write Material Random=true
    Module.Write Position=false
    Module.Write Scale=true
    Module.Write SpriteRotation=false
    Module.Write SpriteSize=false
   )
   /SetVariables(Particles.MeshOrientation=&MakeQuatFromAxisAngle(Axis=<0 1 0>, AngleInDegrees=-90)) {isSetParameters=true}
   &Color(Color=&LerpLinearColors(StartColor=$User.LootLevel_Color, EndColor=<rgba 5 5 5 1>, LerpFactor=0.05))
   &InitialMeshOrientation(Use Orientation Vector=true, Use Rotation Vector=false)
  ]
  particleUpdateScript [
   &ParticleState()
   &ScaleColor(
    Scale Alpha=&FloatFromCurve(
     FloatCurve=DataInterface() {propertyValues {Curve {keys [<at 0 0 RCIM_Cubic>,<at 0.05 1>,<at 0.8 1 RCIM_Cubic>,<at 1 0>]}}}
     Scale Curve=&Multiply_Float(
      A=&Lerp_Float(
       A=0
       B=0.6
       Alpha=&MakeCustomFloatFromBool(Bool=&SetBoolByFloatComparison(A=1, B=0), True Float=1, False Float=0)
      )
      B=&Multiply_Float(
       A=&Add_Float(
        A=&Multiply_Float(
         A=&Multiply_Float(
          A=0.2
          B=&MakeCustomFloatFromBool(Bool=&SetBoolByFloatComparison(A=$User.TurnOnBlink, B=0), True Float=$Emitter.OpacityOverride, False Float=1)
         )
         B=&ClampFloat(Float=$User.Alpha)
        )
        B=0.15
       )
       B=&ClampFloat(Float=$User.Alpha)
      )
     )
    )
    ScaleA=true
    ScaleRGB=false
   )
   &CameraOffset(Camera Offset Amount=20)
   &DynamicMaterialParameters(
    Index 0 Param 1=50
    Index 0 Param 2=1
    Index 0 Param 3=1
    Index 0 Param 4=1
    Module.Param0WriteEnabled=true
    Module.Param1WriteEnabled=true
    Module.Param2WriteEnabled=true
    Module.Param3WriteEnabled=true
   )
   &SkeletalMeshLocation(
    MaskSpawnGroup=false
    Mesh Sampling Type=::NewEnumerator5
    Orientation Sampling=::NewEnumerator1
    Sampled Position Offset=<0 0 -5>
    Skeletal Mesh=$User.SkeletalMesh
    Socket Sampling Mode=::NewEnumerator5
   )
   /SetVariables(Particles.UserColor:LinearColor=&MakeLinearColorFromVectorAndFloat(
     "Vector (RGB)"=<1 1 1>
     "Float (Alpha)"=&Multiply_Float(
      A=&Add_Float(
       A=&Multiply_Float(
        A=&Multiply_Float(
         A=1
         B=&MakeCustomFloatFromBool(Bool=&SetBoolByFloatComparison(A=$User.TurnOnBlink, B=0), True Float=$Emitter.OpacityOverride, False Float=1)
        )
        B=&ClampFloat(Float=$User.Alpha)
       )
       B=0.1
      )
      B=&ClampFloat(Float=$User.Alpha)
     )
    )) {isSetParameters=true}
  ]
  rendererClasses [&NiagaraMeshRendererProperties]
  renderers [{rendererClass=&NiagaraMeshRendererProperties, rendererIndex=0}]
 }
 LightCone001 {
  bEnabled=false
  emitterUpdateScript [
   &EmitterState()
   &SpawnBurst_Instantaneous(Spawn Count=1, Use Spawn Probability=false)
   /SetVariables(Emitter.OpacityOverride:NiagaraFloat=&Multiply_Float(
     A=0.5
     B=&Add_Float(
      A=1
      B=&Sine(
       Normalized Angle=&Multiply_Float(A=$Engine.Time, B=&Multiply_Float(A=$User.BlinkSpeed, B=1))
       Period=1
       Scale=1
       Bias=0
      )
     )
    )) {isSetParameters=true}
  ]
  particleSpawnScript [
   &InitializeParticle(
    Lifetime=7
    Mesh Scale=<0.08 1 1>
    Module.Write Color=false
    Module.Write Lifetime=true
    Module.Write Mass=true
    Module.Write Material Random=true
    Module.Write Position=false
    Module.Write Scale=true
    Module.Write SpriteRotation=false
    Module.Write SpriteSize=false
   )
   /SetVariables(Particles.MeshOrientation=&MakeQuatFromAxisAngle(Axis=<0 1 0>, AngleInDegrees=-90)) {isSetParameters=true}
   &Color(Color=&LerpLinearColors(StartColor=$User.LootLevel_Color, EndColor=<rgba 1 1 1 1>, LerpFactor=0.05))
   &InitialMeshOrientation(Use Orientation Vector=true, Use Rotation Vector=false)
  ]
  particleUpdateScript [
   &ParticleState()
   &ScaleColor(
    Scale Alpha=&FloatFromCurve(
     FloatCurve=DataInterface() {propertyValues {Curve {keys [<at 0 0 RCIM_Cubic>, <at 0.05 0.35>, {arriveTangent=-0.9474, interpMode=RCIM_Cubic, leaveTangent=-0.9474, time=0.8, value=0.35}, <at 1 0>]}}}
     Scale Curve=&Multiply_Float(
      A=5
      B=&Multiply_Float(
       A=&Add_Float(
        A=&Multiply_Float(
         A=&Multiply_Float(
          A=0.1
          B=&MakeCustomFloatFromBool(Bool=&SetBoolByFloatComparison(A=$User.TurnOnBlink, B=0), True Float=$Emitter.OpacityOverride, False Float=1)
         )
         B=&ClampFloat(Float=$User.Alpha)
        )
        B=0.15
       )
       B=&ClampFloat(Float=$User.Alpha)
      )
     )
    )
    ScaleA=true
    ScaleRGB=false
   ) {module=ScaleColor001}
   &CameraOffset(Camera Offset Amount=20)
   &DynamicMaterialParameters(
    Index 0 Param 1=35
    Index 0 Param 2=1
    Index 0 Param 3=1
    Index 0 Param 4=1
    Module.Param0WriteEnabled=true
    Module.Param1WriteEnabled=true
    Module.Param2WriteEnabled=true
    Module.Param3WriteEnabled=true
   )
   &SkeletalMeshLocation(
    MaskSpawnGroup=false
    Mesh Sampling Type=::NewEnumerator5
    Orientation Sampling=::NewEnumerator1
    Skeletal Mesh=$User.SkeletalMesh
    Socket Sampling Mode=::NewEnumerator5
   )
  ]
  rendererClasses [&NiagaraMeshRendererProperties]
  renderers [{rendererClass=&NiagaraMeshRendererProperties, rendererIndex=0}]
 }
 LightCone_mesh {
  bEnabled=false
  emitterUpdateScript [
   &EmitterState(Life Cycle Mode=::NewEnumerator1, Loop Behavior=::NewEnumerator1, Loop Duration Mode=::NewEnumerator1, UseLoopDelay=false)
   &SpawnBurst_Instantaneous(Spawn Count=1, Use Spawn Probability=false)
   /SetVariables(
    Emitter.LightColor:LinearColor=&MultiplyLinearColor(Linear Color=$User.LootLevel_Color, Scale Factor=<1 1 1 1>)
    Emitter.MeshColor:LinearColor=&MultiplyLinearColor(Linear Color=$User.LootLevel_Color, Scale Factor=<0.1 0.1 0.1 1>)
    Emitter.SpriteColor:LinearColor=&MultiplyLinearColor(Linear Color=$User.LootLevel_Color, Scale Factor=<1 1 1 0.1>)
   ) {isSetParameters=true}
   /SetVariables(Emitter.OpacityOverride:NiagaraFloat=&Multiply_Float(
     A=0.5
     B=&Add_Float(
      A=1
      B=&Sine(
       Normalized Angle=&Multiply_Float(A=$Engine.Time, B=&Multiply_Float(A=$User.BlinkSpeed, B=1))
       Period=1
       Scale=1
       Bias=0
      )
     )
    )) {isSetParameters=true}
   /SetVariables(Emitter.MeshOpacity:NiagaraFloat=&Multiply_Float(
     A=10
     B=&Multiply_Float(
      A=&Add_Float(
       A=&Multiply_Float(
        A=&Multiply_Float(
         A=0.1
         B=&MakeCustomFloatFromBool(Bool=&SetBoolByFloatComparison(A=$User.TurnOnBlink, B=0), True Float=$Emitter.OpacityOverride, False Float=1)
        )
        B=&ClampFloat(Float=$User.Alpha)
       )
       B=0.15
      )
      B=&ClampFloat(Float=$User.Alpha)
     )
    )) {isSetParameters=true}
  ]
  particleSpawnScript [
   /Niagara/Modules/Spawn/Initialization/V2/InitializeParticle(
    Color=$Emitter.MeshColor
    Lifetime=2
    Mesh Scale=<0.001 0.001 0.5>
    Mesh Scale Mode=::NewEnumerator4
    Sprite Size Mode=::NewEnumerator3
    Uniform Sprite Size=80
    UsePositionOffset=false
    Write Material Random=true
   ) {module=InitializeParticle001}
   &InitialMeshOrientation(
    Orientation Axis=<0 0 1>
    Orientation Coordinate Space=::Local
    Orientation Vector=<0 0 0>
    Use Orientation Vector=true
    Use Rotation Vector=true
   )
   /SetVariables(Particles.SpriteInitialColor:LinearColor=$Emitter.SpriteColor) {isSetParameters=true}
   /SetVariables(Particles.SpriteLocation:Vector3f=&Add_Vector(A=Linked() {linkedVariable {name=Particles.Position, type=NiagaraPosition}}, B=<0 -12 -20>)) {isSetParameters=true}
   /SetVariables(Particles.SpriteCameraOffset=20) {isSetParameters=true}
   &SkeletalMeshLocation(
    MaskSpawnGroup=false
    Mesh Sampling Type=::NewEnumerator5
    Sampled Position Offset=<0 0 -3>
    Skeletal Mesh=$User.SkeletalMesh
    Socket Sampling Mode=::NewEnumerator5
   )
  ]
  particleUpdateScript [
   &ParticleState(Kill Particles When Lifetime Has Elapsed=false, Let Infinitely Lived Particles Die When Emitter Deactivates=true)
   &ScaleColor(
    Scale Alpha=&FloatFromCurve(FloatCurve=DataInterface() {propertyValues {Curve {keys [
         {interpMode=RCIM_Cubic, tangentMode=RCTM_User, time=0, value=0}
         {arriveTangent=-0.0331, interpMode=RCIM_Cubic, leaveTangent=-0.0331, tangentMode=RCTM_User, time=0.5, value=1}
         {interpMode=RCIM_Cubic, tangentMode=RCTM_User, time=1.0011, value=0.002}
        ]}}})
    ScaleA=true
    ScaleRGB=true
   ) {enabled=false}
   &Color(Color=$Emitter.MeshColor)
   &ScaleColor(
    Color Value To Scale=$Particles.Color
    Scale Alpha=&FloatFromCurve(
     FloatCurve=DataInterface() {propertyValues {Curve {keys [<at 0 0 RCIM_Cubic>, <at 0.25 1>, {arriveTangent=-1.3333, interpMode=RCIM_Cubic, leaveTangent=-1.3333, time=0.75, value=1}, <at 1 0>]}}}
     Scale Curve=&Multiply_Float(
      A=0.5
      B=&Multiply_Float(
       A=&Add_Float(
        A=&Multiply_Float(
         A=&Multiply_Float(
          A=0.1
          B=&MakeCustomFloatFromBool(Bool=&SetBoolByFloatComparison(A=$User.TurnOnBlink, B=0), True Float=$Emitter.OpacityOverride, False Float=1)
         )
         B=&ClampFloat(Float=$User.Alpha)
        )
        B=0.15
       )
       B=&ClampFloat(Float=$User.Alpha)
      )
     )
    )
    ScaleA=true
    ScaleRGB=true
   ) {module=ScaleColor001}
   &FN_Light_Attributes(
    Color=$Emitter.LightColor
    Create Color Attribute=true
    Create Exponent Attribute=true
    Create Position Attribute=true
    Create Radius Attribute=true
    Enable Particle Light Enable=true
    Exponent=3
    Light Enabled=true
    Position=<0 0 15>
    Radius=20
    Scale Color=6
   ) {enabled=false}
  ]
  rendererClasses [
   &NiagaraMeshRendererProperties
   &NiagaraLightRendererProperties
   &NiagaraSpriteRendererProperties
  ]
  renderers [
   {rendererClass=&NiagaraMeshRendererProperties, rendererIndex=0}
   {rendererClass=&NiagaraLightRendererProperties, rendererIndex=1}
   {rendererClass=&NiagaraSpriteRendererProperties, rendererIndex=2}
  ]
 }
 Ray_Soft {
  emitterUpdateScript [
   &EmitterState()
   &SpawnBurst_Instantaneous(Spawn Count=1, Use Spawn Probability=false)
   /SetVariables(Emitter.Color:LinearColor=&MultiplyLinearColor(Linear Color=$User.LootLevel_Color, Scale Factor=<1 1 1 1>)) {isSetParameters=true}
   /SetVariables(Emitter.OpacityOverride:NiagaraFloat=&Multiply_Float(
     A=0.5
     B=&Add_Float(
      A=1
      B=&Sine(
       Normalized Angle=&Multiply_Float(A=$Engine.Time, B=&Multiply_Float(A=$User.BlinkSpeed, B=1))
       Period=1
       Scale=1
       Bias=0
      )
     )
    )) {isSetParameters=true}
  ]
  particleSpawnScript [
   &InitializeParticle(Lifetime=7, Mesh Scale=<0.27 0.225 0.225>, Mesh Scale Mode=::NewEnumerator4, UsePositionOffset=false, Write Material Random=true)
   &InitialMeshOrientation(Rotation=<0 -0.25 0>, Use Orientation Vector=true, Use Rotation Vector=true) {module=InitialMeshOrientation001}
  ]
  particleUpdateScript [
   &ParticleState()
   &SkeletalMeshLocation(
    MaskSpawnGroup=false
    Mesh Sampling Type=::NewEnumerator5
    Sampled Position Offset=<0 0 -4>
    Skeletal Mesh=$User.SkeletalMesh
    Socket Sampling Mode=::NewEnumerator5
   )
   &SolveForcesAndVelocity(Clamp Velocity=false, Limit Acceleration=false)
   /SetVariables(Particles.UserColor:LinearColor=&MakeLinearColorFromVectorAndFloat(
     "Vector (RGB)"=&MakeVectorFromLinearColorRGB(Color=$User.LootLevel_Color)
     "Float (Alpha)"=&Multiply_Float(
      A=&Add_Float(
       A=&Multiply_Float(
        A=&Multiply_Float(
         A=&FloatFromCurve(FloatCurve=DataInterface() {propertyValues {Curve {keys [{interpMode=RCIM_Cubic, tangentMode=RCTM_User, time=0.8, value=1}, {interpMode=RCIM_Cubic, tangentMode=RCTM_User, time=1, value=0}]}}})
         B=&MakeCustomFloatFromBool(Bool=&SetBoolByFloatComparison(A=$User.TurnOnBlink, B=0), True Float=$Emitter.OpacityOverride, False Float=1)
        )
        B=&ClampFloat(Float=$User.Alpha)
       )
       B=0.1
      )
      B=&ClampFloat(Float=&Multiply_Float(
        A=$User.Alpha
        B=&FloatFromCurve(FloatCurve=DataInterface() {propertyValues {Curve {keys [<at 0 0.5>,<at 0.75 0.5>,<at 1 0>]}}}, CurveIndex=$System.NormalizedLoopAge)
       ))
     )
    )) {isSetParameters=true}
   &OrientMeshToVector(Override Quat=false, UseOptionalSideAxis=false, UseOptionalUpAxis=false) {enabled=false, module=OrientMeshToVector001}
  ]
  rendererClasses [&NiagaraMeshRendererProperties]
  renderers [{rendererClass=&NiagaraMeshRendererProperties, rendererIndex=0}]
 }
 Rays {
  emitterUpdateScript [
   &EmitterState()
   &SpawnRate(SpawnRate=10, Use Spawn Probability=false)
   /SetVariables(Emitter.Color:LinearColor=$User.LootLevel_Color) {isSetParameters=true}
   /SetVariables(Emitter.OpacityOverride:NiagaraFloat=&Multiply_Float(
     A=0.5
     B=&Add_Float(
      A=1
      B=&Sine(
       Normalized Angle=&Multiply_Float(A=$Engine.Time, B=&Multiply_Float(A=$User.BlinkSpeed, B=1))
       Period=1
       Scale=1
       Bias=0
      )
     )
    )) {isSetParameters=true}
  ]
  particleSpawnScript [
   &InitializeParticle(
    Color=&RandomRangeLinearColor(Minimum=<rgba 1 1 1 0>, Maximum=<rgba 1 1 1 0>)
    Lifetime Max=0.8
    Lifetime Min=0.6
    Lifetime Mode=::NewEnumerator1
    Mesh Scale Max=<0.265 0.05 0.03>
    Mesh Scale Min=<0.26 0.04 0.02>
    Mesh Scale Mode=::NewEnumerator2
    UsePositionOffset=false
    Write Material Random=true
   ) {module=InitializeParticle001}
   &DynamicMaterialParameters(
    Index 0 Param 1=&UniformRangedFloat(Minimum=-1, Maximum=1)
    Index 0 Param 2=1
    Module.Param0WriteEnabled=true
    Module.Param1WriteEnabled=true
    Module.Param2WriteEnabled=true
    Module.Param3WriteEnabled=true
   )
   &InitialMeshOrientation(
    Rotation=&Add_Vector(A=<0 0 0>, B=&RandomRangeVector(Minimum=<0 0 0>, Maximum=<1 0 0>))
    Use Orientation Vector=true
    Use Rotation Vector=true
   ) {module=InitialMeshOrientation001}
   &InitialMeshOrientation(
    Rotation=&RandomRangeVector(Minimum=<0 0 -0.08>, Maximum=<0 0 0.08>)
    Use Orientation Vector=true
    Use Rotation Vector=true
   )
   &InitialMeshOrientation(
    Rotation=&RandomRangeVector(Minimum=<0 -0.17 0>, Maximum=<0 -0.33 0>)
    Use Orientation Vector=true
    Use Rotation Vector=true
   ) {module=InitialMeshOrientation003}
  ]
  particleUpdateScript [
   &ParticleState()
   &OrientMeshToVector(Override Quat=false, UseOptionalSideAxis=false, UseOptionalUpAxis=false) {module=OrientMeshToVector001}
   &SkeletalMeshLocation(
    MaskSpawnGroup=false
    Mesh Sampling Type=::NewEnumerator5
    Sampled Position Offset=<0 0 -5>
    Skeletal Mesh=$User.SkeletalMesh
    Socket Sampling Mode=::NewEnumerator5
   )
   &ScaleMeshSize(Scale Factor=&MakeVector(
     X=&FloatFromCurve(FloatCurve=DataInterface() {propertyValues {Curve {keys [<at 0 0.8>,<at 1 1.2>]}}})
     Y=1
     Z=&FloatFromCurve(FloatCurve=DataInterface() {propertyValues {Curve {keys [<at 0 1.8>,<at 1 1>]}}})
    ))
   /SetVariables(Particles.UserColor:LinearColor=&MakeLinearColorFromVectorAndFloat(
     "Vector (RGB)"=&MakeVectorFromLinearColorRGB(Color=$User.LootLevel_Color)
     "Float (Alpha)"=&Multiply_Float(
      A=&Add_Float(
       A=&Multiply_Float(
        A=&Multiply_Float(
         A=&FloatFromCurve(FloatCurve=DataInterface() {propertyValues {Curve {keys [{interpMode=RCIM_Cubic, tangentMode=RCTM_User, time=0.8, value=1}, {interpMode=RCIM_Cubic, tangentMode=RCTM_User, time=1, value=0}]}}})
         B=&MakeCustomFloatFromBool(Bool=&SetBoolByFloatComparison(A=$User.TurnOnBlink, B=0), True Float=$Emitter.OpacityOverride, False Float=1)
        )
        B=&ClampFloat(Float=$User.Alpha)
       )
       B=0.1
      )
      B=&ClampFloat(Float=&Multiply_Float(
        A=$User.Alpha
        B=&FloatFromCurve(FloatCurve=DataInterface() {propertyValues {Curve {keys [<at 0 1>,<at 0.75 1>,<at 1 0>]}}}, CurveIndex=$System.NormalizedLoopAge)
       ))
     )
    )) {isSetParameters=true}
   &SolveForcesAndVelocity(Clamp Velocity=false, Limit Acceleration=false) {enabled=false}
  ]
  rendererClasses [&NiagaraMeshRendererProperties]
  renderers [{rendererClass=&NiagaraMeshRendererProperties, rendererIndex=0}]
 }
}
summary {
 systemName=NS_Skyfire_LightCone
 emitters [
  {emitterName=BaseLight, rendererClasses [&NiagaraSpriteRendererProperties, &NiagaraLightRendererProperties]}
  {
   bEnabled=false
   emitterName=Light
   rendererClasses [&NiagaraLightRendererProperties, &NiagaraComponentRendererProperties]
  }
  {bEnabled=false, emitterName=LightCone, rendererClasses [&NiagaraMeshRendererProperties]}
  {bEnabled=false, emitterName=LightCone001, rendererClasses [&NiagaraMeshRendererProperties]}
  {
   bEnabled=false
   emitterName=LightCone_mesh
   rendererClasses [
    &NiagaraMeshRendererProperties
    &NiagaraLightRendererProperties
    &NiagaraSpriteRendererProperties
   ]
  }
  {emitterName=BaseLight_AngleRelevant, rendererClasses [&NiagaraSpriteRendererProperties]}
  {emitterName=Rays, rendererClasses [&NiagaraMeshRendererProperties]}
  {emitterName=Cone_Ray, rendererClasses [&NiagaraMeshRendererProperties]}
  {emitterName=Ray_Soft, rendererClasses [&NiagaraMeshRendererProperties]}
 ]
 userVariables [
  User.Alpha {type=NiagaraFloat, defaultValue=1, description=""}
  User.BlinkSpeed {type=NiagaraFloat, defaultValue=1, description=""}
  User.FX_Target_Position {type=Vector3f, description="", defaultValue {struct=Vector3f, value=<0 0 1>}}
  User.LootLevel {type=NiagaraFloat, defaultValue=0, description=""}
  User.LootLevel_Color {type=LinearColor, description="", defaultValue {struct=LinearColor, value=<rgba 0.8994 0.0377 2 1>}}
  User.Scale_Size {type=NiagaraFloat, defaultValue=0, description=""}
  User.SkeletalMesh {
   type=NiagaraDataInterfaceSkeletalMesh
   defaultValue {
    struct=NiagaraExt_VariableValue_DataInterface
    value {dataInterface=&self:NiagaraDataInterfaceSkeletalMesh_1}
   }
  }
  User.TurnOnBlink {type=NiagaraFloat, defaultValue=0, description=""}
 ]
}
